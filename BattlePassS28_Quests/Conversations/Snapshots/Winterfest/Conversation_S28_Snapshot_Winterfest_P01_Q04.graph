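# Conversation database for the Season 28 Winterfest snapshot quest P1 Q04, where Holiday Boxy briefs the player on destroying Krampus's hacked drones.
# A single tagged entry point leads into a linear chain of four speech lines.
# The first choice auto-selects as main questline dialogue when the NPC has the holiday Cardboard Crew tag, the quest is active and its first objective is incomplete.
# The last choice carries a quest-specific tag.
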ConversationDatabase /BattlePassS28_Quests/Conversations/Snapshots/Winterfest/Conversation_S28_Snapshot_Winterfest_P01_Q04
entry_points [{tag=Conversation.Entry.Season28.HolidayBoxy.P01.Q04.1, destinations [#1]}]
nodes [
 ConversationEntryPointNode_0 {
  id=#1
  class=ConversationEntryPointNode
  outputs [#2]
  properties {entryTag {tagName=Conversation.Entry.Season28.HolidayBoxy.P01.Q04.1}}
 }
 Speech_C_1 {
  id=#2
  class=Speech_C
  outputs [#3]
  properties {
   softTaskIcon=&T_UI_NPC_MainQuest
   generalConfig {defaultMessage="It looks like Krampus is behind the present thievery!"}
  }
  sub_nodes [
   ConversationChoiceNode_2 {
    class=ConversationChoiceNode
    properties {
     defaultChoiceDisplayText=Talk to Holiday Boxy
     choiceTags {gameplayTags [{tagName=Conversation.Choice.AutoSelectIfValid}, {tagName=Conversation.Choice.Category.MainQuestline}]}
    }
   }
   FortConversationRequirement_ControllerRequirement_1 {
    class=FortConversationRequirement_ControllerRequirement
    properties {
     nodeName=Require: NPC Has Conversation Tag
     participantId {tagName=Conversation.Participant.NPC}
     requirement {class=FortControllerRequirement_NPCHasConversationEntryTag, properties {queryTag {tagName=Conversation.Entry.CardboardCrew_Holiday}}}
    }
   }
   FortConversationRequirement_ControllerRequirement_3 {
    class=FortConversationRequirement_ControllerRequirement
    properties {
     nodeName=Require: Is Quest Active
     requirement {
      class=FortControllerRequirement_IsQuestActive
      properties {questDefinition=/BattlePassS28_Quests/Items/QuestItems/Winterfest/Winterfest_Snapshot/Quest_S28_Winterfest_Snapshot_P1_Q04}
     }
    }
   }
   FortConversationRequirement_ControllerRequirement_5 {
    class=FortConversationRequirement_ControllerRequirement
    properties {
     nodeName=Require: Has Completed Quest Objective
     requirement {
      class=FortControllerRequirement_HasCompletedQuestObjective
      properties {
       bInverseCondition=true
       objectiveBackendName=Quest_S28_Winterfest_Snapshot_P1_Q04_obj0
       questDefinition=&Quest_S28_Winterfest_Snapshot_P1_Q04
      }
     }
    }
   }
  ]
 }
 Speech_C_28 {
  id=#3
  class=Speech_C
  outputs [#4]
  properties {generalConfig {defaultMessage="He's even hacked Ship It! drones to help steal packages."}}
  sub_nodes [ConversationChoiceNode_1 {class=ConversationChoiceNode, properties {defaultChoiceDisplayText=Continue}}]
 }
 Speech_C_2 {
  id=#4
  class=Speech_C
  outputs [#5]
  properties {generalConfig {defaultMessage=Destroy the hacked drones and get some eyes on Krampus's whereabouts.}}
  sub_nodes [ConversationChoiceNode_3 {class=ConversationChoiceNode, properties {defaultChoiceDisplayText=Continue}}]
 }
 Speech_C_0 {
  id=#5
  class=Speech_C
  outputs []
  properties {generalConfig {defaultMessage="Let's take back Winterfest!"}}
  sub_nodes [ConversationChoiceNode_4 {
    class=ConversationChoiceNode
    properties {defaultChoiceDisplayText=Continue, choiceTags {gameplayTags [{tagName=Athena.Quests.S28.Conversation.HolidayBoxy.Choice.P01.Q04}]}}
   }]
 }
]
